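; Gameplay ability that tracks quest progress for distance floated while the player holds the zombie chicken (NugZ).
; On event activation it checks the holding tag, starts a randomized looping timer, and ends when the NugFloat active tag is removed.
; Each tick adds whole meters moved in 3D and, above a threshold, sends the running total as a quest stat.
; The total resets only at ability end, which also clears timer and task.
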(event Ability|EventActivateAbilityFromEvent
  (params (EventData :type Gameplay%20Event%20Data%20Structure%20%28by%20ref%29))
  (bind _owningplayerpawn (Variables|GAVars|Data|GetOwningPlayerPawn))
  ;[1 1 1] Activate ability
  (|CallParentFunction EventData)
  (if (GameplayTags|HasMatchingGameplayTag _owningplayerpawn "(TagName=\"Granted.Athena.Holding.NugZ\")")
    (bind _asynctaskproxy (Ability|Tasks|WaitGameplayTagRemove "(TagName=\"Athena.Quests.NugFloat.Active\")")
      (:then
        (Variables|Default|SetTrackFloatTask _asynctaskproxy)
        (bind _output_get (Variables|GAVars|Data|SetPreviousLocation (Transformation|GetActorLocation _owningplayerpawn)))
        (Variables|GAVars|Data|SetCurrentLocation _output_get)
        (bind _output_get_2 (Variables|GAVars|Setup|SetInterval (Math|Random|RandomFloatinRange 1.7 2.5)))
        (bind _returnvalue (Utilities|Time|SetTimerbyEvent (delegate Custom|SendDistanceUpdate) _output_get_2 true))
        (Variables|GAVars|Data|SetTImerHandle _returnvalue))
      (:Removed
        (Ability|EndAbility)))
    (else
      (Ability|EndAbility))))

(event Custom|SendDistanceUpdate
  (bind (_previouslocation_x _previouslocation_y _previouslocation_z) (Math|Vector|BreakVector (Variables|GetPreviousLocation_X)))
  (bind (_currentlocation_x _currentlocation_y _currentlocation_z) (Math|Vector|BreakVector (Variables|GetCurrentLocation_X)))
  ;[1 1 1] Send Distance Update
  ;;[1 1 1] Cache distance traveled (horizontal and vertical)
  (Variables|GAVars|Data|SetCurrentLocation (Transformation|GetActorLocation (Variables|GAVars|Data|GetOwningPlayerPawn)))
  (Variables|GAVars|Data|SetMetersTraveled (+ (Math|Float|Floor (/ (Math|Vector|VectorLength (Math|Vector|vector-vector (Math|Vector|MakeVector _currentlocation_x _currentlocation_y _currentlocation_z) (Math|Vector|MakeVector _previouslocation_x _previouslocation_y _previouslocation_z))) 1e+02)) (Variables|GAVars|Data|GetMetersTraveled)))
  (Variables|GAVars|Data|SetPreviousLocation (Variables|GAVars|Data|GetCurrentLocation)) ; Reset location
  ;;[1 1 1] Send Quest Progress
  (if (> (Variables|GAVars|Data|GetMetersTraveled) (Variables|Default|GetMetersToUpdateQuest))
    (Quest|SendComplexCustomStatEvent (Variables|GAVars|Data|GetOwningPlayerQuestManager) 0 (Variables|GAVars|Setup|GetT_Quest) "(GameplayTags=)" (Variables|GAVars|Data|GetMetersTraveled))
    (Development|PrintString "Complex custom sent" true true "(R=1.000000,G=0.000000,B=0.565984,A=1.000000)")
    (else
      (Development|PrintString "meters travelled not great enough" true true "(R=1.000000,G=0.000000,B=0.565984,A=1.000000)"))))

(event Ability|EventOnEndAbility
  (params (bWasCancelled :type Boolean))
  (bind _timerhandle (Variables|GAVars|Data|GetTImerHandle))
  ;[1 1 1] On End Ability
  (|CallParentFunction bWasCancelled)
  (Variables|GAVars|Data|SetMetersTraveled 0)
  (GameplayTasks|EndTask (Variables|Default|GetTrackFloatTask))
  (if (Utilities|Time|IsValidTimerHandle _timerhandle)
    (Utilities|Time|ClearandInvalidateTimerbyHandle _timerhandle)))
